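# Looping gadget-class sound cue for the PuffyStraw active state, judging by its name.
# A single looping wave player feeds an Enveloper whose volume curve fades the sound in from near silence over roughly two seconds.
# It uses PuffyStraw-specific attenuation, keeps playing virtually when silent, and overrides concurrency to allow one instance with volume scaling.
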SoundCue /NPCLibrary/Audio/PuffyStraw/PuffyStraw_ActiveLoop_Cue
root=Enveloper_0
nodes [
 {
  id=Enveloper_0
  children [WavePlayer_1]
  properties {pitchCurve {editorCurveData {keys [{time=0}]}}, volumeCurve {editorCurveData {keys [<at 0.0355 0.038>,<at 2.2587 1.0028>]}}}
 }
 {id=WavePlayer_1, properties {bLooping=true, soundWaveAssetPtr=&/PuffyStraw_Active_Loop_02}}
]
properties {
 attenuationSettings=&/PuffyStraw_Att
 bOverrideConcurrency=true
 maxDistance=3800
 soundClassObject=&Fort_Gadgets
 virtualizationMode=PlayWhenSilent
 volumeMultiplier=0.35
 concurrencyOverrides {maxCount=1, volumeScale=0.5}
}
