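# Sound cue for the PuffyStraw deploy sound in the NPCLibrary audio set, routed to the Fort_Gadgets sound class.
# A single wave player plays PuffyStraw_Deploy_01 directly, with no randomization or mixing.
# It uses PuffyStraw attenuation with a slight volume boost and overrides concurrency with a small voice limit, while also referencing the Evil_Boss_Con concurrency set from TinyBalloon.
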SoundCue /NPCLibrary/Audio/PuffyStraw/PuffyStraw_Deploy_Cue
root=WavePlayer_0
nodes [{id=WavePlayer_0, properties {soundWaveAssetPtr=&/PuffyStraw_Deploy_01}}]
properties {
 attenuationSettings=&/PuffyStraw_Att
 bOverrideConcurrency=true
 maxDistance=3800
 soundClassObject=&Fort_Gadgets
 volumeMultiplier=1.1
 concurrencyOverrides {maxCount=2}
 concurrencySet [/NPCLibrary/Audio/TinyBalloon/Evil_Boss_Con]
}
