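# Sound cue for the destroyed sound of PuffyStraw in the NPC library.
# A single Wave Player feeds the output directly with one destruction wave, without randomization or modulation.
# It uses the PuffyStraw attenuation asset and the Fort_Gadgets sound class.
# A concurrency override caps simultaneous instances at two, taking precedence over the referenced Evil_Boss_Con concurrency set from the TinyBalloon folder.
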SoundCue /NPCLibrary/Audio/PuffyStraw/PuffyStraw_Destroyed_Cue
root=WavePlayer_1
nodes [{id=WavePlayer_1, properties {soundWaveAssetPtr=&/PuffyStraw_Destroyed_01}}]
properties {
 attenuationSettings=&/PuffyStraw_Att
 bOverrideConcurrency=true
 maxDistance=3800
 soundClassObject=&Fort_Gadgets
 concurrencyOverrides {maxCount=2}
 concurrencySet [/NPCLibrary/Audio/TinyBalloon/Evil_Boss_Con]
}
