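# Sound cue for PuffyStraw damage impacts, routed to the impacts sound class with bullet impact attenuation.
# A quality-level node feeds the first two levels from an eleven-variant random pool and the last two from a four-variant pool that is excluded from branch culling.
# A root modulator randomly raises pitch on the chosen wave, and the cue boosts volume and overrides concurrency to cap simultaneous voices.
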SoundCue /NPCLibrary/Audio/PuffyStraw/PuffyStraw_Impact_Damage_Cue
root=Modulator_0
nodes [
 {id=Modulator_0, children [QualityLevel_0], properties {pitchMax=1.4, pitchMin=1.1}}
 {id=QualityLevel_0, children [Random_3,Random_3,Random_4,Random_4]}
 {
  id=Random_3
  children [
   WavePlayer_7
   WavePlayer_8
   WavePlayer_9
   WavePlayer_10
   WavePlayer_14
   WavePlayer_15
   WavePlayer_16
   WavePlayer_17
   WavePlayer_18
   WavePlayer_19
   WavePlayer_20
  ]
  properties {weights [1,1,1,1,1,1,1,1,1,1,1]}
 }
 {
  id=Random_4
  children [WavePlayer_21, WavePlayer_22, WavePlayer_23, WavePlayer_24]
  properties {bShouldExcludeFromBranchCulling=true, weights [1,1,1,1]}
 }
 {id=WavePlayer_10, properties {soundWaveAssetPtr=&/PuffyStraw_Impact_Damage_02}}
 {id=WavePlayer_14, properties {soundWaveAssetPtr=&/PuffyStraw_Impact_Damage_03}}
 {id=WavePlayer_15, properties {soundWaveAssetPtr=&/PuffyStraw_Impact_Damage_04}}
 {id=WavePlayer_16, properties {soundWaveAssetPtr=&/PuffyStraw_Impact_Damage_05}}
 {id=WavePlayer_17, properties {soundWaveAssetPtr=&/PuffyStraw_Impact_Damage_06}}
 {id=WavePlayer_18, properties {soundWaveAssetPtr=&/PuffyStraw_Impact_Damage_07}}
 {id=WavePlayer_19, properties {soundWaveAssetPtr=&/PuffyStraw_Impact_Damage_08}}
 {id=WavePlayer_20, properties {soundWaveAssetPtr=&/PuffyStraw_Impact_Damage_09}}
 {id=WavePlayer_21, properties {soundWaveAssetPtr=&/PuffyStraw_Impact_Damage_01}}
 {id=WavePlayer_22, properties {soundWaveAssetPtr=&PuffyStraw_Impact_Damage_02}}
 {id=WavePlayer_23, properties {soundWaveAssetPtr=&PuffyStraw_Impact_Damage_03}}
 {id=WavePlayer_24, properties {soundWaveAssetPtr=&PuffyStraw_Impact_Damage_04}}
 {id=WavePlayer_7, properties {soundWaveAssetPtr=&/PuffyStraw_Impact_Damage_10}}
 {id=WavePlayer_8, properties {soundWaveAssetPtr=&/PuffyStraw_Impact_Damage_11}}
 {id=WavePlayer_9, properties {soundWaveAssetPtr=&PuffyStraw_Impact_Damage_01}}
]
properties {
 attenuationSettings=/Game/Sounds/Attenuation/Fort_Bullet_Impact_Att
 bOverrideConcurrency=true
 maxDistance=5300
 soundClassObject=&Fort_Impacts
 volumeMultiplier=2.8224
 attenuationOverrides {falloffDistance=5000, attenuationShapeExtents {x=1000}}
 concurrencyOverrides {maxCount=6}
 concurrencySet [&RifleBulletImpact_Conc]
}
