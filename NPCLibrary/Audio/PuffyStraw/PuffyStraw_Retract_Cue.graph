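# PuffyStraw_Retract_Cue is a Sound Cue for the PuffyStraw NPC library audio, named for a retract action.
# A single wave player feeds PuffyStraw_Retract_02 straight to the output, with no randomization or modulation.
# It uses the PuffyStraw attenuation asset and the Fort_Gadgets sound class at slightly reduced volume.
# A concurrency override caps it at two voices, superseding the referenced TinyBalloon Evil_Boss_Con concurrency set.
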SoundCue /NPCLibrary/Audio/PuffyStraw/PuffyStraw_Retract_Cue
root=WavePlayer_3
nodes [{id=WavePlayer_3, properties {soundWaveAssetPtr=&/PuffyStraw_Retract_02}}]
properties {
 attenuationSettings=&/PuffyStraw_Att
 bOverrideConcurrency=true
 maxDistance=3800
 soundClassObject=&Fort_Gadgets
 volumeMultiplier=0.8
 concurrencyOverrides {maxCount=2}
 concurrencySet [/NPCLibrary/Audio/TinyBalloon/Evil_Boss_Con]
}
