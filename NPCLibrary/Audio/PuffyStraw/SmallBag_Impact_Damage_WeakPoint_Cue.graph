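# Sound cue for weak-point damage impacts on the PuffyStraw NPC's small bag, set up as a bullet-impact sound.
# A quality switch sends its first two levels to an eight-variant random pool and its last two to a four-variant subset, which is exempt from branch culling.
# A modulator randomly raises pitch on each play.
# Playback uses bullet-impact attenuation, the impacts sound class and capped concurrency.
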SoundCue /NPCLibrary/Audio/PuffyStraw/SmallBag_Impact_Damage_WeakPoint_Cue
root=Modulator_0
nodes [
 {id=Modulator_0, children [QualityLevel_0], properties {pitchMax=1.4, pitchMin=1.1}}
 {id=QualityLevel_0, children [Random_0,Random_0,Random_1,Random_1]}
 {
  id=Random_0
  children [WavePlayer_1, WavePlayer_2, WavePlayer_3, WavePlayer_4, WavePlayer_5, WavePlayer_6, WavePlayer_11, WavePlayer_12]
  properties {weights [1,1,1,1,1,1,1,1]}
 }
 {
  id=Random_1
  children [WavePlayer_7, WavePlayer_25, WavePlayer_26, WavePlayer_27]
  properties {bShouldExcludeFromBranchCulling=true, weights [1,1,1,1]}
 }
 {id=WavePlayer_1, properties {soundWaveAssetPtr=&/SmallBag_Impact_Damage_WeakPoint_02}}
 {id=WavePlayer_11, properties {soundWaveAssetPtr=&/SmallBag_Impact_Damage_WeakPoint_08}}
 {id=WavePlayer_12, properties {soundWaveAssetPtr=&/SmallBag_Impact_Damage_WeakPoint_09}}
 {id=WavePlayer_2, properties {soundWaveAssetPtr=&/SmallBag_Impact_Damage_WeakPoint_03}}
 {id=WavePlayer_25, properties {soundWaveAssetPtr=&SmallBag_Impact_Damage_WeakPoint_02}}
 {id=WavePlayer_26, properties {soundWaveAssetPtr=&SmallBag_Impact_Damage_WeakPoint_03}}
 {id=WavePlayer_27, properties {soundWaveAssetPtr=&/SmallBag_Impact_Damage_WeakPoint_04}}
 {id=WavePlayer_3, properties {soundWaveAssetPtr=&SmallBag_Impact_Damage_WeakPoint_04}}
 {id=WavePlayer_4, properties {soundWaveAssetPtr=&/SmallBag_Impact_Damage_WeakPoint_05}}
 {id=WavePlayer_5, properties {soundWaveAssetPtr=&/SmallBag_Impact_Damage_WeakPoint_06}}
 {id=WavePlayer_6, properties {soundWaveAssetPtr=&/SmallBag_Impact_Damage_WeakPoint_07}}
 {id=WavePlayer_7, properties {soundWaveAssetPtr=&SmallBag_Impact_Damage_WeakPoint_05}}
]
properties {
 attenuationSettings=/Game/Sounds/Attenuation/Fort_Bullet_Impact_Att
 bOverrideConcurrency=true
 maxDistance=5300
 soundClassObject=&Fort_Impacts
 volumeMultiplier=3.4272
 attenuationOverrides {falloffDistance=5000, attenuationShapeExtents {x=1000}}
 concurrencyOverrides {maxCount=6}
 concurrencySet [&RifleBulletImpact_Conc]
}
