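# Sound cue for the start of the SmallBag PuffyStraw cast, an NPC library sound routed to the melee weapon sound class.
# A mixer layers a reused BrawnySmasher pre-transition sound at full level with a quieter PuffyStraw cast-start wave.
# The PuffyStraw layer starts after a short fixed delay and plays at a slightly lowered, non-randomized pitch.
# The cue uses PuffyStraw attenuation and a boosted overall volume.
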SoundCue /NPCLibrary/Audio/PuffyStraw/SmallBag_PuffyStraw_Cast_Start_Cue
root=Mixer_1
nodes [
 {id=Delay_1, children [WavePlayer_3], properties {delayMax=0.1, delayMin=0.1}}
 {id=Mixer_1, children [WavePlayer_0,Modulator_1], properties {inputVolume [1,0.35]}}
 {id=Modulator_1, children [Delay_1], properties {pitchMax=0.9, pitchMin=0.9, volumeMax=1, volumeMin=1}}
 {id=WavePlayer_0, properties {soundWaveAssetPtr=/Game/Athena/Sounds/NPCs/BrawnySmasher/BrawnySmasher_PreTransition_Short_01}}
 {id=WavePlayer_3, properties {soundWaveAssetPtr=&/PuffyStraw_Cast_Start_01}}
]
properties {
 attenuationSettings=&/PuffyStraw_Att
 maxDistance=3800
 soundClassObject=&Fort_WeaponMelee_PSM
 volumeMultiplier=1.25
}
